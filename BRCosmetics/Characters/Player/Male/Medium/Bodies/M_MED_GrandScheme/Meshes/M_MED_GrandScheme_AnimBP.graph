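# AnimBlueprint for the GrandScheme male medium body cosmetic, copying the attached parent's pose and adjusting armor and gear bones.
# The live chain runs a Control Rig, then belt, knee and chestplate rotation constraints, with crouch knee offsets.
# A backpack-attach position constraint is active below LOD 2, despite its LowLOD alpha name.
# Weapon-specific shoulder-armor blends and their arm-angle alpha logic exist but connect to neither Output Pose nor an event.
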AnimBlueprint /BRCosmetics/Characters/Player/Male/Medium/Bodies/M_MED_GrandScheme/Meshes/M_MED_GrandScheme_AnimBP
target_skeleton=&/M_MED_GrandScheme_Skeleton
graphs [
 AnimGraph {nodes [
   n1 = Get Assault_Equip() -> (Assault_Equip:double)
   n2 = "Transform (Modify) Bone - Bone: dyn_shoulder_armor_l_MSTR"(
    ComponentPose=Local To Component(LocalPose=Copy Pose From Mesh() {bUseAttachedParent=true})
    Alpha=@n1
    Translation="0.000,-1.214,-0.715"
    Rotation="2,0,0"
    Scale="1,1,1"
   ) {
    rotationMode=BMM_Replace
    rotationSpace=BCS_BoneSpace
    scaleMode=BMM_Replace
    scaleSpace=BCS_BoneSpace
    translationMode=BMM_Replace
    translationSpace=BCS_BoneSpace
    alphaBoolBlend {blendOption=Linear}
    boneToModify {boneName=dyn_shoulder_armor_l_MSTR}
    rotation {pitch=2}
    translation=<0 -1.2145 -0.7154>
   }
   n3 = "Transform (Modify) Bone - Bone: dyn_shoulder_armor_r_MSTR"(ComponentPose=@n2, Alpha=@n1, Translation="0,0,0", Rotation="0,0,0", Scale="1,1,1") {
    rotationMode=BMM_Replace
    rotationSpace=BCS_BoneSpace
    scaleMode=BMM_Replace
    scaleSpace=BCS_BoneSpace
    translationMode=BMM_Replace
    translationSpace=BCS_BoneSpace
    alphaBoolBlend {blendOption=Linear}
    alphaScaleBiasClamp {interpSpeedDecreasing=1, interpSpeedIncreasing=1}
    boneToModify {boneName=dyn_shoulder_armor_r_MSTR}
   }
   n4 = "Transform (Modify) Bone - Bone: dyn_shoulder_armor_l_MSTR"(
    ComponentPose=Local To Component(LocalPose=Copy Pose From Mesh() {bUseAttachedParent=true})
    Alpha=Get ADS_Down_L_Assault() -> (ADS_Down_L_Assault:double)
    Translation="0,-2,-2"
    Rotation="20,20,0"
    Scale="1,1,1"
   ) {
    rotationMode=BMM_Replace
    rotationSpace=BCS_BoneSpace
    scaleMode=BMM_Replace
    scaleSpace=BCS_BoneSpace
    translationMode=BMM_Replace
    translationSpace=BCS_BoneSpace
    alphaBoolBlend {blendOption=Linear}
    boneToModify {boneName=dyn_shoulder_armor_l_MSTR}
    rotation {pitch=20, yaw=20}
    translation {y=-2, z=-2}
   }
   n5 = "Transform (Modify) Bone - Bone: dyn_shoulder_armor_r_MSTR"(ComponentPose=@n4, Alpha=Get ADS_Down_R_Assault() -> (ADS_Down_R_Assault:double), Translation="3,1,-3", Rotation="15,0,30", Scale="1,1,1") {
    rotationMode=BMM_Replace
    rotationSpace=BCS_BoneSpace
    scaleMode=BMM_Replace
    scaleSpace=BCS_BoneSpace
    translationMode=BMM_Replace
    translationSpace=BCS_BoneSpace
    alphaBoolBlend {blendOption=Linear}
    boneToModify {boneName=dyn_shoulder_armor_r_MSTR}
    rotation {pitch=15, roll=30}
    translation=<3 1 -3>
   }
   n6 = "Transform (Modify) Bone - Bone: dyn_shoulder_armor_l_MSTR"(ComponentPose=@n5, Alpha=Get ADS_UP_L_Assault() -> (ADS_UP_L_Assault:double), Translation="-2,-3,2", Rotation="-15,0,-30", Scale="1,1,1") {
    rotationMode=BMM_Replace
    rotationSpace=BCS_BoneSpace
    scaleMode=BMM_Replace
    scaleSpace=BCS_BoneSpace
    translationMode=BMM_Replace
    translationSpace=BCS_BoneSpace
    alphaBoolBlend {blendOption=Linear}
    boneToModify {boneName=dyn_shoulder_armor_l_MSTR}
    rotation {pitch=-15, roll=-30}
    translation=<-2 -3 2>
   }
   n7 = "Transform (Modify) Bone - Bone: dyn_shoulder_armor_r_MSTR"(ComponentPose=@n6, Alpha=Get ADS_UP_R_Assault() -> (ADS_UP_R_Assault:double), Translation="-1,-2,-2", Rotation="13,10,0", Scale="1,1,1") {
    rotationMode=BMM_Replace
    rotationSpace=BCS_BoneSpace
    scaleMode=BMM_Replace
    scaleSpace=BCS_BoneSpace
    translationMode=BMM_Replace
    translationSpace=BCS_BoneSpace
    alphaBoolBlend {blendOption=Linear}
    boneToModify {boneName=dyn_shoulder_armor_r_MSTR}
    rotation {pitch=13, yaw=10}
    translation=<-1 -2 -2>
   }
   n8 = Blend Poses by bool(
    BlendPose_0=Component To Local(ComponentPose=@n7)
    BlendPose_1=Component To Local(ComponentPose=@n3)
    BlendTime_1=0.3
    bActiveValue=Get bIsTargeting()
   ) {blendType=Linear}
   n9 = Get Shotgun_Equip() -> (Shotgun_Equip:double)
   n10 = "Transform (Modify) Bone - Bone: dyn_shoulder_armor_l_MSTR"(
    ComponentPose=Local To Component(LocalPose=Copy Pose From Mesh() {bUseAttachedParent=true})
    Alpha=@n9
    Translation="0,0,0"
    Rotation="0,0,0"
    Scale="1,1,1"
   ) {
    rotationMode=BMM_Replace
    rotationSpace=BCS_BoneSpace
    scaleMode=BMM_Replace
    scaleSpace=BCS_BoneSpace
    translationMode=BMM_Replace
    translationSpace=BCS_BoneSpace
    alphaBoolBlend {blendOption=Linear}
    boneToModify {boneName=dyn_shoulder_armor_l_MSTR}
   }
   n11 = "Transform (Modify) Bone - Bone: dyn_shoulder_armor_r_MSTR"(ComponentPose=@n10, Alpha=@n9, Translation="0,0,0", Rotation="0,0,0", Scale="1,1,1") {
    rotationMode=BMM_Replace
    rotationSpace=BCS_BoneSpace
    scaleMode=BMM_Replace
    scaleSpace=BCS_BoneSpace
    translationMode=BMM_Replace
    translationSpace=BCS_BoneSpace
    alphaBoolBlend {blendOption=Linear}
    alphaScaleBiasClamp {interpSpeedDecreasing=1, interpSpeedIncreasing=1}
    boneToModify {boneName=dyn_shoulder_armor_r_MSTR}
   }
   n12 = "Transform (Modify) Bone - Bone: dyn_shoulder_armor_l_MSTR"(
    ComponentPose=Local To Component(LocalPose=Copy Pose From Mesh() {bUseAttachedParent=true})
    Alpha=Get ADS_Down_L_Shotgun() -> (ADS_Down_L_Shotgun:double)
    Translation="2,0,0"
    Rotation="15,0,0"
    Scale="1,1,1"
   ) {
    rotationMode=BMM_Replace
    rotationSpace=BCS_BoneSpace
    scaleMode=BMM_Replace
    scaleSpace=BCS_BoneSpace
    translationMode=BMM_Replace
    translationSpace=BCS_BoneSpace
    alphaBoolBlend {blendOption=Linear}
    boneToModify {boneName=dyn_shoulder_armor_l_MSTR}
    rotation {pitch=15}
    translation {x=2}
   }
   n13 = "Transform (Modify) Bone - Bone: dyn_shoulder_armor_r_MSTR"(ComponentPose=@n12, Alpha=Get ADS_Down_R_Shotgun() -> (ADS_Down_R_Shotgun:double), Translation="3,0,0", Rotation="0,0,0", Scale="1,1,1") {
    rotationMode=BMM_Replace
    rotationSpace=BCS_BoneSpace
    scaleMode=BMM_Replace
    scaleSpace=BCS_BoneSpace
    translationMode=BMM_Replace
    translationSpace=BCS_BoneSpace
    alphaBoolBlend {blendOption=Linear}
    boneToModify {boneName=dyn_shoulder_armor_r_MSTR}
    translation {x=3}
   }
   n14 = "Transform (Modify) Bone - Bone: dyn_shoulder_armor_l_MSTR"(
    ComponentPose=@n13
    Alpha=Get ADS_UP_L_Shotgun() -> (ADS_UP_L_Shotgun:double)
    Translation="-0.151,-5.504,0.342"
    Rotation="-8.421,20.693,-53.320"
    Scale="1,1,1"
   ) {
    rotationMode=BMM_Replace
    rotationSpace=BCS_BoneSpace
    scaleMode=BMM_Replace
    scaleSpace=BCS_BoneSpace
    translationMode=BMM_Replace
    translationSpace=BCS_BoneSpace
    alphaBoolBlend {blendOption=Linear}
    boneToModify {boneName=dyn_shoulder_armor_l_MSTR}
    rotation=<rot -8.4209 20.693 -53.32>
    translation=<-0.151 -5.504 0.342>
   }
   n15 = "Transform (Modify) Bone - Bone: dyn_shoulder_armor_r_MSTR"(ComponentPose=@n14, Alpha=Get ADS_UP_R_Shotgun() -> (ADS_UP_R_Shotgun:double), Translation="-1,-2,-1", Rotation="5,10,10", Scale="1,1,1") {
    rotationMode=BMM_Replace
    rotationSpace=BCS_BoneSpace
    scaleMode=BMM_Replace
    scaleSpace=BCS_BoneSpace
    translationMode=BMM_Replace
    translationSpace=BCS_BoneSpace
    alphaBoolBlend {blendOption=Linear}
    boneToModify {boneName=dyn_shoulder_armor_r_MSTR}
    rotation=<rot 5 10 10>
    translation=<-1 -2 -1>
   }
   n16 = Blend Poses by bool(
    BlendPose_0=Blend Poses by bool(
     BlendPose_0=Component To Local(ComponentPose=@n15)
     BlendPose_1=Component To Local(ComponentPose=@n11)
     BlendTime_1=0.3
     bActiveValue=Get bIsTargeting()
    ) {blendType=Linear}
    bActiveValue=Get Shotgun() -> (Shotgun:bool)
   ) {blendType=Linear}
   n17 = Get Assault() -> (Assault:bool)
   n18 = "Transform (Modify) Bone - Bone: dyn_shoulder_armor_l_MSTR"(
    ComponentPose=Local To Component(LocalPose=Copy Pose From Mesh() {bUseAttachedParent=true})
    Alpha=Get ADS_Down_L_Pistol() -> (ADS_Down_L_Pistol:double)
    Translation="2,0,0"
    Rotation="0,0,0"
    Scale="1,1,1"
   ) {
    rotationMode=BMM_Additive
    rotationSpace=BCS_BoneSpace
    scaleMode=BMM_Additive
    scaleSpace=BCS_BoneSpace
    translationMode=BMM_Additive
    translationSpace=BCS_BoneSpace
    alphaBoolBlend {blendOption=Linear}
    boneToModify {boneName=dyn_shoulder_armor_l_MSTR}
    translation {x=2}
   }
   n19 = "Transform (Modify) Bone - Bone: dyn_shoulder_armor_r_MSTR"(ComponentPose=@n18, Alpha=Get ADS_Down_R_Pistol() -> (ADS_Down_R_Pistol:double), Translation="1.75,1,-3", Rotation="15,0,-15", Scale="1,1,1") {
    rotationMode=BMM_Additive
    rotationSpace=BCS_BoneSpace
    scaleMode=BMM_Additive
    scaleSpace=BCS_BoneSpace
    translationMode=BMM_Additive
    translationSpace=BCS_BoneSpace
    alphaBoolBlend {blendOption=Linear}
    boneToModify {boneName=dyn_shoulder_armor_r_MSTR}
    rotation {pitch=15, roll=-15}
    translation=<1.75 1 -3>
   }
   n20 = "Transform (Modify) Bone - Bone: dyn_shoulder_armor_l"(ComponentPose=@n19, Alpha=Get ADS_Up_L_Pistol() -> (ADS_Up_L_Pistol:double), Translation="-2,-1,0", Rotation="0,0,-45", Scale="1,1,1") {
    rotationMode=BMM_Additive
    rotationSpace=BCS_BoneSpace
    scaleSpace=BCS_BoneSpace
    translationMode=BMM_Additive
    translationSpace=BCS_BoneSpace
    alphaBoolBlend {blendOption=Linear}
    boneToModify {boneName=dyn_shoulder_armor_l}
    rotation {roll=-45}
    translation=<-2 -1>
   }
   n21 = "Transform (Modify) Bone - Bone: dyn_shoulder_armor_r_MSTR"(ComponentPose=@n20, Alpha=Get ADS_Up_R_Pistol() -> (ADS_Up_R_Pistol:double), Translation="1,-2,-2", Rotation="13,10,30", Scale="1,1,1") {
    rotationMode=BMM_Additive
    rotationSpace=BCS_BoneSpace
    scaleSpace=BCS_BoneSpace
    translationMode=BMM_Additive
    translationSpace=BCS_BoneSpace
    alphaBoolBlend {blendOption=Linear}
    boneToModify {boneName=dyn_shoulder_armor_r_MSTR}
    rotation=<rot 13 10 30>
    translation=<1 -2 -2>
   }
   n22 = Component To Local(ComponentPose=@n21)
   n23 = "Get Pistol_Equip "() -> ("Pistol_Equip ":double)
   n24 = "Transform (Modify) Bone - Bone: dyn_shoulder_armor_l_MSTR"(
    ComponentPose=Local To Component(LocalPose=Copy Pose From Mesh() {bUseAttachedParent=true})
    Alpha=@n23
    Translation="0,0,-1"
    Rotation="5.000,0.000,-20.000"
    Scale="1,1,1"
   ) {
    rotationMode=BMM_Replace
    rotationSpace=BCS_BoneSpace
    scaleMode=BMM_Additive
    scaleSpace=BCS_BoneSpace
    translationMode=BMM_Replace
    translationSpace=BCS_BoneSpace
    alphaBoolBlend {blendOption=Linear}
    boneToModify {boneName=dyn_shoulder_armor_l_MSTR}
    rotation=<rot 5 0 -20>
    translation {z=-1}
   }
   n25 = "Transform (Modify) Bone - Bone: dyn_shoulder_armor_r_MSTR"(ComponentPose=@n24, Alpha=@n23, Translation="0,0.192,-0.883", Rotation="2.000,-0.000,20.000", Scale="1,1,1") {
    rotationMode=BMM_Additive
    rotationSpace=BCS_BoneSpace
    scaleMode=BMM_Additive
    scaleSpace=BCS_BoneSpace
    translationMode=BMM_Additive
    translationSpace=BCS_BoneSpace
    alphaBoolBlend {blendOption=Linear}
    alphaScaleBiasClamp {interpSpeedDecreasing=1, interpSpeedIncreasing=1}
    boneToModify {boneName=dyn_shoulder_armor_r_MSTR}
    rotation=<rot 2.0001 -0.0001 20>
    translation {y=0.192, z=-0.883}
   }
   n26 = Component To Local(ComponentPose=@n25)
   n27 = Get bIsTargeting()
   n28 = "Transform (Modify) Bone - Bone: dyn_shoulder_armor_l_MSTR"(
    ComponentPose=Local To Component(LocalPose=Copy Pose From Mesh() {bUseAttachedParent=true})
    Alpha=1
    Translation="0,0,0"
    Rotation="0,0,0"
    Scale="1,1,1"
   ) {
    rotationMode=BMM_Additive
    rotationSpace=BCS_BoneSpace
    scaleMode=BMM_Additive
    scaleSpace=BCS_BoneSpace
    translationMode=BMM_Additive
    translationSpace=BCS_BoneSpace
    alphaBoolBlend {blendOption=Linear}
    boneToModify {boneName=dyn_shoulder_armor_l_MSTR}
   }
   n29 = "Transform (Modify) Bone - Bone: dyn_shoulder_armor_r_MSTR"(ComponentPose=@n28, Alpha=1, Translation="0,0,0", Rotation="0,0,0", Scale="1,1,1") {
    rotationMode=BMM_Additive
    rotationSpace=BCS_BoneSpace
    scaleMode=BMM_Additive
    scaleSpace=BCS_BoneSpace
    translationMode=BMM_Replace
    translationSpace=BCS_BoneSpace
    alphaBoolBlend {blendOption=Linear}
    boneToModify {boneName=dyn_shoulder_armor_r_MSTR}
   }
   n30 = "Transform (Modify) Bone - Bone: dyn_shoulder_armor_l"(ComponentPose=@n29, Alpha=1, Translation="0,0,0", Rotation="0,0,0", Scale="1,1,1") {
    rotationMode=BMM_Additive
    rotationSpace=BCS_BoneSpace
    scaleSpace=BCS_BoneSpace
    translationMode=BMM_Additive
    translationSpace=BCS_BoneSpace
    alphaBoolBlend {blendOption=Linear}
    boneToModify {boneName=dyn_shoulder_armor_l}
   }
   n31 = "Transform (Modify) Bone - Bone: dyn_shoulder_armor_r"(ComponentPose=@n30, Alpha=1, Translation="0,0,0", Rotation="0,0,0", Scale="1,1,1") {
    rotationMode=BMM_Additive
    rotationSpace=BCS_BoneSpace
    scaleSpace=BCS_BoneSpace
    translationMode=BMM_Additive
    translationSpace=BCS_BoneSpace
    alphaBoolBlend {blendOption=Linear}
    boneToModify {boneName=dyn_shoulder_armor_r}
   }
   Blend Poses by bool(
    BlendPose_0=Blend Poses by bool(
     BlendPose_0=Blend Poses by bool(BlendPose_0=@n22, BlendPose_1=@n26, BlendTime_1=0.3, bActiveValue=@n27) {blendType=Linear}
     BlendPose_1=Blend Poses by bool(BlendPose_0=@n8, BlendPose_1=@n16, bActiveValue=@n17) {blendType=Linear}
     bActiveValue=Get Pistol() -> (Pistol:bool)
    ) {blendType=Linear}
    BlendPose_1=Component To Local(ComponentPose=@n31)
    BlendTime_1=0.3
    bActiveValue=Get Alt_weapon() -> (Alt_weapon:bool)
   ) {blendType=Linear}
   n32 = Component To Local()
   n33 = Component To Local()
   n34 = Get bIsSkyDiving()
   n35 = "Constraint - dyn_chest (RR)"!(
    ComponentPose=Local To Component(LocalPose=Blend Poses by bool(BlendPose_0=@n32, BlendPose_1=@n33, bActiveValue=@n34) {blendType=Linear})
    Alpha=1
    ConstraintWeights_0=0.1
    ConstraintWeights_1:float=0.9
   ) -> (Pose) {
    alphaBoolBlend {blendOption=Linear}
    boneToModify {boneName=dyn_chest}
    constraintSetup [{transformType=Rotation, targetBone {boneName=spine_05}}, {transformType=Rotation, targetBone {boneName=spine_04}}]
    constraintWeights [1,1]
   }
   Component To Local(ComponentPose=@n35)
   "Constraint - dyn_shoulder_strap_r_MSTR (P)"(Alpha=@n23, ConstraintWeights_0=1) {
    alphaBoolBlend {blendOption=Linear}
    boneToModify {boneName=dyn_shoulder_strap_r_MSTR}
    constraintSetup [{targetBone {boneName=clavicle_r}}]
    constraintWeights [1]
   }
   n36 = "Constraint - dyn_shoulder_armor_l_MSTR (R)"(Alpha=1, ConstraintWeights_0=1) {
    alphaBoolBlend {blendOption=Linear}
    boneToModify {boneName=dyn_shoulder_armor_l_MSTR}
    constraintSetup [{transformType=Rotation, targetBone {boneName=upperarm_l}}]
    constraintWeights [1]
   }
   "Constraint - dyn_shoulder_armor_r_MSTR (R)"(ComponentPose=@n36, Alpha=1, ConstraintWeights_0=1) {
    alphaBoolBlend {blendOption=Linear}
    boneToModify {boneName=dyn_shoulder_armor_r_MSTR}
    constraintSetup [{transformType=Rotation, targetBone {boneName=upperarm_r}}]
    constraintWeights [1]
   }
   n37 = "Transform (Modify) Bone - Bone: dyn_shoulder_armor_l"(
    ComponentPose=Local To Component(LocalPose=Copy Pose From Mesh() {bUseAttachedParent=true})
    Alpha=1
    Translation="1.5,0,3.5"
    Rotation="-25,0,0"
    Scale="1,1,1"
   ) {
    rotationMode=BMM_Additive
    rotationSpace=BCS_BoneSpace
    scaleMode=BMM_Additive
    scaleSpace=BCS_BoneSpace
    translationMode=BMM_Additive
    translationSpace=BCS_BoneSpace
    alphaBoolBlend {blendOption=Linear}
    boneToModify {boneName=dyn_shoulder_armor_l}
    rotation {pitch=-25}
    translation {x=1.5, z=3.5}
   }
   "Transform (Modify) Bone - Bone: dyn_shoulder_armor_r"(ComponentPose=@n37, Alpha=1, Translation="-1.5,0,-3.5", Rotation="-25,0,0", Scale="1,1,1") {
    rotationMode=BMM_Additive
    rotationSpace=BCS_BoneSpace
    scaleMode=BMM_Additive
    scaleSpace=BCS_BoneSpace
    translationMode=BMM_Replace
    translationSpace=BCS_BoneSpace
    alphaBoolBlend {blendOption=Linear}
    boneToModify {boneName=dyn_shoulder_armor_r}
    rotation {pitch=-25}
    translation {x=-1.5, z=-3.5}
   }
   "Transform (Modify) Bone - Bone: dyn_shoulder_armor_r_MSTR"(Alpha=1, Translation="-1,0,-2", Rotation="5,0,0", Scale="1,1,1") {
    rotationMode=BMM_Replace
    rotationSpace=BCS_BoneSpace
    scaleMode=BMM_Additive
    scaleSpace=BCS_BoneSpace
    translationMode=BMM_Replace
    translationSpace=BCS_BoneSpace
    alphaBoolBlend {blendOption=Linear}
    alphaScaleBiasClamp {interpSpeedDecreasing=1, interpSpeedIncreasing=1}
    boneToModify {boneName=dyn_shoulder_armor_r_MSTR}
    rotation {pitch=5}
    translation {x=-1, z=-2}
   }
   n38 = "Transform (Modify) Bone - Bone: dyn_shoulder_armor_l"(Alpha=1, Translation="0,0,0", Rotation="0,0,0", Scale="1,1,1") {
    rotationMode=BMM_Replace
    rotationSpace=BCS_BoneSpace
    scaleSpace=BCS_BoneSpace
    translationMode=BMM_Replace
    translationSpace=BCS_BoneSpace
    alphaBoolBlend {blendOption=Linear}
    boneToModify {boneName=dyn_shoulder_armor_l}
   }
   "Transform (Modify) Bone - Bone: dyn_shoulder_armor_r"(ComponentPose=@n38, Alpha=1, Translation="0,0,0", Rotation="0,0,0", Scale="1,1,1") {
    rotationMode=BMM_Replace
    rotationSpace=BCS_BoneSpace
    scaleSpace=BCS_BoneSpace
    translationMode=BMM_Replace
    translationSpace=BCS_BoneSpace
    alphaBoolBlend {blendOption=Linear}
    boneToModify {boneName=dyn_shoulder_armor_r}
   }
   n39 = Control Rig(Source=Copy Pose From Mesh() {bUseAttachedParent=true}) {
    alphaBoolBlend {blendOption=Linear}
    controlRigAssetReference {blueprintRigClass=&/M_MED_Grand_Scheme_Main_CtrlRig.M_MED_Grand_Scheme_Main_CtrlRig_C}
   }
   n40 = "Constraint - dyn_belt_main (R)"(ComponentPose=Local To Component(LocalPose=@n39), Alpha=1, ConstraintWeights_0=1) {
    alphaBoolBlend {blendOption=Linear}
    boneToModify {boneName=dyn_belt_main}
    constraintSetup [{transformType=Rotation, targetBone {boneName=pelvis}}]
    constraintWeights [1]
   }
   n41 = "Constraint - dyn_master_knee_l (RR)"!(ComponentPose=@n40, Alpha=1, ConstraintWeights_0=0.2, ConstraintWeights_1:float=0.8) -> (Pose) {
    alphaBoolBlend {blendOption=Linear}
    boneToModify {boneName=dyn_master_knee_l}
    constraintSetup [{transformType=Rotation, targetBone {boneName=thigh_l}}, {transformType=Rotation, targetBone {boneName=calf_l}}]
    constraintWeights [1,1]
   }
   n42 = Get bIsCrouching()
   n43 = "Transform (Modify) Bone - Bone: dyn_master_knee_l"!(ComponentPose=@n41, bAlphaBoolEnabled:bool=@n42, Translation="3,-2,0", Rotation="0,0,0", Scale="1,1,1") -> (Pose) {
    alphaInputType=Bool
    translationMode=BMM_Additive
    translationSpace=BCS_ParentBoneSpace
    alphaBoolBlend {blendOption=Linear}
    boneToModify {boneName=dyn_master_knee_l}
    translation=<3 -2>
   }
   n44 = "Constraint - dyn_master_knee_r (RR)"!(ComponentPose=@n43, Alpha=1, ConstraintWeights_0=0.2, ConstraintWeights_1:float=0.8) -> (Pose) {
    alphaBoolBlend {blendOption=Linear}
    boneToModify {boneName=dyn_master_knee_r}
    constraintSetup [{transformType=Rotation, targetBone {boneName=thigh_r}}, {transformType=Rotation, targetBone {boneName=calf_r}}]
    constraintWeights [1,1]
   }
   n45 = "Transform (Modify) Bone - Bone: dyn_master_knee_r"!(ComponentPose=@n44, bAlphaBoolEnabled:bool=@n42, Translation="-3,2,0", Rotation="0,0,0", Scale="1,1,1") -> (Pose) {
    alphaInputType=Bool
    translationMode=BMM_Additive
    translationSpace=BCS_ParentBoneSpace
    alphaBoolBlend {blendOption=Linear}
    boneToModify {boneName=dyn_master_knee_r}
    translation=<-3 2>
   }
   n46 = "Constraint - chestplate_master (RR)"!(ComponentPose=@n45, Alpha=1, ConstraintWeights_0=0.5, ConstraintWeights_1:float=0.5) -> (Pose) {
    alphaBoolBlend {blendOption=Linear}
    boneToModify {boneName=chestplate_master}
    constraintSetup [{transformType=Rotation, targetBone {boneName=spine_03}}, {transformType=Rotation, targetBone {boneName=spine_04}}]
    constraintWeights [1,1]
   }
   n47 = "Constraint - attach_backpack (PP)"!(ComponentPose=@n46, Alpha=Get LowLOD() -> (LowLOD:double), ConstraintWeights_0=0.5, ConstraintWeights_1:float=0.5) -> (Pose) {
    alphaBoolBlend {blendOption=Linear}
    boneToModify {boneName=attach_backpack}
    constraintSetup [{targetBone {boneName=dyn_chest_armor_bk_r}}, {targetBone {boneName=dyn_chest_armor_bk_l}}]
    constraintWeights [1,1]
   }
   Output Pose(Result=Component To Local(ComponentPose=@n47)) {layerGroup=None}
   Two Way Blend!(A, B, bAlphaBoolEnabled:bool=True) -> (Pose) {blendNode {alphaInputType=Bool, alphaBoolBlend {blendOption=Linear}}}
  ]}
 EventGraph {nodes [
   Event BlueprintUpdateAnimation()
   > n1 = Branch(Condition="integer >= integer"(A=Get CurrentLOD(), B=2))
   > Set LowLOD(LowLOD:double=0) -> (Output_Get:double)
   Set LowLOD(execute=@n1.else, LowLOD=1)
   Event BlueprintInitializeAnimation()
   > Set Skel Mesh(Skel Mesh=GetOwnerSkeletalMeshComponent())
   n2 = Sequence()
   > n3 = Branch(Condition=Get bIsSkyDiving())
   Set ADS_Down_L_Pistol(execute=[@n3.then, @n10.then], ADS_Down_L_Pistol:double=0) -> (Output_Get:double)
   > Set ADS_Down_R_Pistol(ADS_Down_R_Pistol:double=0) -> (Output_Get:double)
   > Set ADS_Up_L_Pistol(ADS_Up_L_Pistol:double=0) -> (Output_Get:double)
   > Set ADS_Up_R_Pistol(ADS_Up_R_Pistol:double=0) -> (Output_Get:double)
   > Set ADS_Down_L_Assault(ADS_Down_L_Assault:double=0) -> (Output_Get:double)
   > Set ADS_Down_R_Assault(ADS_Down_R_Assault:double=0) -> (Output_Get:double)
   > Set ADS_UP_L_Assault(ADS_UP_L_Assault:double=0) -> (Output_Get:double)
   > Set ADS_UP_R_Assault(ADS_UP_R_Assault:double=0) -> (Output_Get:double)
   > Set ADS_Down_L_Shotgun(ADS_Down_L_Shotgun:double=0) -> (Output_Get:double)
   > Set ADS_Down_R_Shotgun(ADS_Down_R_Shotgun:double=0) -> (Output_Get:double)
   > Set ADS_UP_L_Shotgun(ADS_UP_L_Shotgun:double=0) -> (Output_Get:double)
   > Set ADS_UP_R_Shotgun(ADS_UP_R_Shotgun:double=0) -> (Output_Get:double)
   > "Set Pistol_Equip "("Pistol_Equip ":double=0) -> (Output_Get:double)
   > Set Assault_Equip(Assault_Equip:double=0) -> (Output_Get:double)
   > Set Shotgun_Equip(Shotgun_Equip:double=0) -> (Output_Get:double)
   n4 = Cast To FortPlayerAnimInstance(execute=@n3.else, Object=GetAnimInstance(self=GetOwnerSkeletalMeshComponent()))
   > Set Alt_weapon(Alt_weapon:bool=false) -> (Output_Get:bool)
   > n5 = Switch on EFortWeaponCoreAnimation!(execute, Selection:EFortWeaponCoreAnimation Enum=Get WeaponCoreAnim(self=@n4)) -> (
    Melee:Exec
    Pistol:Exec
    Shotgun:Exec
    PaperBlueprint:Exec
    Rifle:Exec
    MeleeOneHand:Exec
    MachinePistol:Exec
    RocketLauncher:Exec
    GrenadeLauncher:Exec
    GoingCommando:Exec
    AssaultRifle:Exec
    TacticalShotgun:Exec
    SniperRifle:Exec
    TrapPlacement:Exec
    ShoulderLauncher:Exec
    AbilityDecoTool:Exec
    Crossbow:Exec
    C4:Exec
    RemoteControl:Exec
    DualWield:Exec
    AR_BullPup:Exec
    AR_ForwardGrip:Exec
    MedPackPaddles:Exec
    SMG_P90:Exec
    AR_DrumGun:Exec
    Consumable_Small:Exec
    Consumable_Large:Exec
    Balloon:Exec
    MountedTurret:Exec
    CreativeTool:Exec
    ExplosiveBow:Exec
    AshtonIndigo:Exec
    AshtonChicago:Exec
    MeleeDualWield:Exec
    Unarmed:Exec
   )
   Set Assault_Equip(execute=[@n5.AR_ForwardGrip, @n5.SMG_P90, @n5.Shotgun, @n5.TacticalShotgun], Assault_Equip=0)
   > "Set Pistol_Equip "("Pistol_Equip "=0)
   > Set Shotgun_Equip(Shotgun_Equip=1)
   > Set Alt_weapon(Alt_weapon=true)
   > Set Shotgun(Shotgun:bool=true) -> (Output_Get:bool)
   > Set Pistol(Pistol:bool=false) -> (Output_Get:bool)
   > Set Assault(Assault:bool=false) -> (Output_Get:bool)
   > n6 = Sequence!(execute) -> (then_0, then_1, then_2:Exec, then_3:Exec, then_4:Exec)
   n7 = Get Skel Mesh()
   > Set ADS_Down_L_Shotgun(ADS_Down_L_Shotgun=GetAlphaFromReferenceTransformDelta(SkeletalMeshComponent=@n7, BoneName=upperarm_l, BaseName=clavicle_l, InputRange=(X=35,Y=20), TwistAxis="0,0,1"))
   > Set ADS_Down_L_Pistol(ADS_Down_L_Pistol=0)
   > Set ADS_Down_L_Assault(ADS_Down_L_Assault=0)
   Set ADS_Down_R_Shotgun(
    execute=@n6.then_1
    ADS_Down_R_Shotgun=GetAlphaFromReferenceTransformDelta(SkeletalMeshComponent=@n7, BoneName=upperarm_r, BaseName=clavicle_r, InputRange=(X=120,Y=40), TwistAxis="0,0,1")
   )
   > Set ADS_Down_R_Pistol(ADS_Down_R_Pistol=0)
   > Set ADS_Down_R_Assault(ADS_Down_R_Assault=0)
   Set ADS_UP_L_Shotgun(
    execute=@n6.then_2
    ADS_UP_L_Shotgun=GetAlphaFromReferenceTransformDelta(SkeletalMeshComponent=@n7, BoneName=upperarm_l, BaseName=clavicle_l, DeltaComponentType=EulerY, InputRange=(X=12,Y=50), TwistAxis="0,0,1")
   )
   > Set ADS_Up_L_Pistol(ADS_Up_L_Pistol=0)
   > Set ADS_UP_L_Assault(ADS_UP_L_Assault=0)
   Set ADS_UP_R_Shotgun(
    execute=@n6.then_3
    ADS_UP_R_Shotgun=GetAlphaFromReferenceTransformDelta(SkeletalMeshComponent=@n7, BoneName=upperarm_r, BaseName=clavicle_r, InputRange=(X=35,Y=51), TwistAxis="0,0,1")
   )
   > Set ADS_Up_R_Pistol(ADS_Up_R_Pistol=0)
   > Set ADS_UP_R_Assault(ADS_UP_R_Assault=0)
   "Set Pistol_Equip "(execute=[@n5.AssaultRifle, @n5.Rifle], "Pistol_Equip "=0)
   > Set Shotgun_Equip(Shotgun_Equip=0)
   > Set Assault_Equip(Assault_Equip=1)
   > Set Alt_weapon(Alt_weapon=true)
   > Set Assault(Assault=true)
   > Set Pistol(Pistol=false)
   > Set Shotgun(Shotgun=false)
   > n8 = Sequence()
   n9 = Get Skel Mesh()
   > Set ADS_Down_L_Assault(ADS_Down_L_Assault=GetAlphaFromReferenceTransformDelta(SkeletalMeshComponent=@n9, BoneName=upperarm_l, BaseName=clavicle_l, InputRange=(X=35,Y=25), TwistAxis="0,0,1"))
   > Set ADS_Down_L_Pistol(ADS_Down_L_Pistol=0)
   > Set ADS_Down_L_Shotgun(ADS_Down_L_Shotgun=0)
   Set ADS_Down_R_Assault(
    execute=@n8.then_1
    ADS_Down_R_Assault=GetAlphaFromReferenceTransformDelta(SkeletalMeshComponent=@n9, BoneName=upperarm_r, BaseName=clavicle_r, InputRange=(X=120,Y=40), TwistAxis="0,0,1")
   )
   > Set ADS_Down_R_Pistol(ADS_Down_R_Pistol=0)
   > Set ADS_Down_R_Shotgun(ADS_Down_R_Shotgun=0)
   Set ADS_UP_L_Assault(
    execute=@n8.then_2
    ADS_UP_L_Assault=GetAlphaFromReferenceTransformDelta(SkeletalMeshComponent=@n9, BoneName=upperarm_l, BaseName=clavicle_l, DeltaComponentType=EulerY, InputRange=(X=25,Y=40), TwistAxis="0,0,1")
   )
   > Set ADS_Up_L_Pistol(ADS_Up_L_Pistol=0)
   > Set ADS_UP_L_Shotgun(ADS_UP_L_Shotgun=0)
   Set ADS_UP_R_Assault(
    execute=@n8.then_3
    ADS_UP_R_Assault=GetAlphaFromReferenceTransformDelta(SkeletalMeshComponent=@n9, BoneName=upperarm_r, BaseName=clavicle_r, InputRange=(X=30,Y=48), TwistAxis="0,0,1")
   )
   > Set ADS_Up_R_Pistol(ADS_Up_R_Pistol=0)
   > Set ADS_UP_R_Shotgun(ADS_UP_R_Shotgun=0)
   Set Pistol(execute=@n5.Melee, Pistol=false)
   > Set Shotgun(Shotgun=false)
   > n10 = Set Assault(Assault=false)
   Set Shotgun_Equip(execute=@n5.Pistol, Shotgun_Equip=0)
   > Set Assault_Equip(Assault_Equip=0)
   > "Set Pistol_Equip "("Pistol_Equip "=1)
   > Set Alt_weapon(Alt_weapon=true)
   > Set Pistol(Pistol=true)
   > Set Shotgun(Shotgun=false)
   > Set Assault(Assault=false)
   > n11 = Sequence()
   n12 = Get Skel Mesh()
   > Set ADS_Down_L_Pistol(ADS_Down_L_Pistol=GetAlphaFromReferenceTransformDelta(SkeletalMeshComponent=@n12, BoneName=upperarm_l, BaseName=clavicle_l, InputRange=(X=84,Y=16), TwistAxis="0,0,1"))
   > Set ADS_Down_L_Assault(ADS_Down_L_Assault=0)
   > Set ADS_Down_L_Shotgun(ADS_Down_L_Shotgun=0)
   Set ADS_Down_R_Pistol(
    execute=@n11.then_1
    ADS_Down_R_Pistol=GetAlphaFromReferenceTransformDelta(SkeletalMeshComponent=@n12, BoneName=upperarm_r, BaseName=clavicle_r, InputRange=(X=110,Y=40), TwistAxis="0,0,1")
   )
   > Set ADS_Down_R_Assault(ADS_Down_R_Assault=0)
   > Set ADS_Down_R_Shotgun(ADS_Down_R_Shotgun=0)
   Set ADS_Up_L_Pistol(
    execute=@n11.then_2
    ADS_Up_L_Pistol=GetAlphaFromReferenceTransformDelta(SkeletalMeshComponent=@n12, BoneName=upperarm_l, BaseName=clavicle_l, InputRange=(X=55,Y=75), TwistAxis="0,0,1")
   )
   > Set ADS_UP_L_Assault(ADS_UP_L_Assault=0)
   > Set ADS_UP_L_Shotgun(ADS_UP_L_Shotgun=0)
   Set ADS_Up_R_Pistol(
    execute=@n11.then_3
    ADS_Up_R_Pistol=GetAlphaFromReferenceTransformDelta(SkeletalMeshComponent=@n12, BoneName=upperarm_r, BaseName=clavicle_r, InputRange=(X=70,Y=120), TwistAxis="0,0,1")
   )
   > Set ADS_UP_R_Assault(ADS_UP_R_Assault=0)
   > Set ADS_UP_R_Shotgun(ADS_UP_R_Shotgun=0)
   n13 = Get PawnMovementDirectionAzimuth() -> (PawnMovementDirectionAzimuth:float)
   n14 = float - float(A=MapRangeClamped(Value=@n13, InRangeA=-45, InRangeB=-90), B=MapRangeClamped(Value=@n13, InRangeA=-90, InRangeB=-135)) {is=title K2Node_CallFunction}
   Set Left_Direct(execute=@n2.then_1, Left_Direct:double=SelectFloat(A="Absolute (Float)"(A=@n14), B=0, bPickA=Get bIsMoving2D())) -> (Output_Get:double)
   n15 = Branch(Condition=Get bIsTargeting())
   Set Shotgun_Equip(execute=@n15.else, Shotgun_Equip=1)
   > Set Alt_weapon(Alt_weapon=true)
   > Set Shotgun(Shotgun=true)
   > Set Assault(Assault=false)
   > Set Pistol(Pistol=false)
   > Set ADS_Down_L_Shotgun(ADS_Down_L_Shotgun=0)
   > Set ADS_Down_R_Shotgun(ADS_Down_R_Shotgun=0)
   > Set ADS_UP_L_Shotgun(ADS_UP_L_Shotgun=0)
   > Set ADS_UP_R_Shotgun(ADS_UP_R_Shotgun=0)
   n16 = Branch(Condition=Get bIsTargeting())
   "Set Pistol_Equip "(execute=@n16.else, "Pistol_Equip "=1)
   > Set Alt_weapon(Alt_weapon=true)
   > Set Pistol(Pistol=true)
   > Set Shotgun(Shotgun=false)
   > Set Assault(Assault=false)
   > Set ADS_Down_L_Pistol(ADS_Down_L_Pistol=0)
   > Set ADS_Down_R_Pistol(ADS_Down_R_Pistol=0)
   > Set ADS_Up_L_Pistol(ADS_Up_L_Pistol=0)
   > Set ADS_Up_R_Pistol(ADS_Up_R_Pistol=0)
   n17 = Branch(Condition=Get bIsTargeting())
   Set Assault_Equip(execute=@n17.else, Assault_Equip=1)
   > Set Alt_weapon(Alt_weapon=true)
   > Set Assault(Assault=true)
   > Set Pistol(Pistol=false)
   > Set Shotgun(Shotgun=false)
   > Set ADS_Down_L_Assault(ADS_Down_L_Assault=0)
   > Set ADS_Down_R_Assault(ADS_Down_R_Assault=0)
   > Set ADS_UP_L_Assault(ADS_UP_L_Assault=0)
   > Set ADS_UP_R_Assault(ADS_UP_R_Assault=0)
   Get Alt_weapon()
  ]}
]
variables [
 ADS_Down_L_Assault {type=number}
 ADS_Down_L_Pistol {type=number}
 ADS_Down_L_Shotgun {type=number}
 ADS_Down_R_Assault {type=number}
 ADS_Down_R_Pistol {type=number}
 ADS_Down_R_Shotgun {type=number}
 ADS_UP_L_Assault {type=number}
 ADS_Up_L_Pistol {type=number}
 ADS_UP_L_Shotgun {type=number}
 ADS_UP_R_Assault {type=number}
 ADS_Up_R_Pistol {type=number}
 ADS_UP_R_Shotgun {type=number}
 Alt_weapon {type=boolean}
 Assault {type=boolean, default=true}
 Assault_Equip {type=number}
 Left_Direct {type=number}
 LowLOD {type=number}
 Pistol {type=boolean}
 "Pistol_Equip " {type=number}
 Shotgun {type=boolean, default=true}
 Shotgun_Equip {type=number}
 Skel Mesh {type=SkeletalMeshComponent}
]
