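# Animation Blueprint for the GrandScheme male character's face accessory, following the parent mesh's pose.
# It pins dyn_head_alt to the head's position and replaces its rotation with head-versus-neck_02 delta, yaw nearly zeroed.
# A front-end idle curve weights a DefaultSlot blend from that bone over the cached pose.
# Emote and phone-booth-exit variables update but drive nothing in the AnimGraph, where a head rotation constraint sits unconnected.
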AnimBlueprint /BRCosmetics/Characters/Player/Male/Medium/Bodies/M_MED_GrandScheme/Meshes/Parts/M_MED_GrandScheme_FaceAcc_AnimBP
target_skeleton=/Game/Characters/Player/Male/Medium/Bodies/M_MED_GrandScheme/Meshes/Parts/M_MED_GrandScheme_FaceAcc_Skeleton
graphs [
 AnimGraph {nodes [
   "Constraint - head (R)"(Alpha=1, ConstraintWeights_0=1) {
    alphaBoolBlend {blendOption=Linear}
    boneToModify {boneName=head}
    constraintSetup [{transformType=Rotation, targetBone {boneName=neck_01}}]
    constraintWeights [1]
   }
   Output Pose(Result=Layered blend per bone(
     BasePose=Use cached pose 'startPose'()
     BlendPoses_0=Slot 'DefaultSlot'()
     BlendWeights_0=Get Custom Idle is Playing() -> (Custom Idle is Playing:double)
    ) {curveBlendOption=NormalizeByWeight, layerSetup [{branchFilters [{boneName=dyn_head_alt}]}]}) {layerGroup=None}
   n1 = "Constraint - dyn_head_alt (P)"(ComponentPose=Local To Component(LocalPose=Copy Pose From Mesh() {bUseAttachedParent=true}), Alpha=1, ConstraintWeights_0=1) {
    alphaBoolBlend {blendOption=Linear}
    alphaScaleBiasClamp {bInterpResult=true, interpSpeedDecreasing=4}
    boneToModify {boneName=dyn_head_alt}
    constraintSetup [{targetBone {boneName=head}}]
    constraintWeights [1]
   }
   n2 = "Transform (Modify) Bone - Bone: dyn_head_alt"(ComponentPose=@n1, Alpha=1, Translation="0,0,0", Rotation=Get HeadRotation() -> (HeadRotation:Rotator), Scale="1,1,1") {
    rotationMode=BMM_Replace
    rotationSpace=BCS_BoneSpace
    translationSpace=BCS_BoneSpace
    alphaBoolBlend {blendOption=Linear}
    alphaScaleBiasClamp {interpSpeedDecreasing=1, interpSpeedIncreasing=1}
    boneToModify {boneName=dyn_head_alt}
   }
   n3 = "Constraint - dyn_head_alt (P)"(ComponentPose=@n2, Alpha=1, ConstraintWeights_0=1) {
    alphaBoolBlend {blendOption=Linear}
    boneToModify {boneName=dyn_head_alt}
    constraintSetup [{targetBone {boneName=head}}]
    constraintWeights [1]
   }
   Save cached pose 'startPose'(Pose=Component To Local(ComponentPose=@n3)) {cacheName=startPose}
  ]}
 EventGraph {nodes [
   Event BlueprintUpdateAnimation()
   > n1 = Sequence()
   n2 = GetOwnerSkeletalMeshComponent()
   n3 = BreakRotator(InRot=BreakTransform(InTransform=GetDeltaTransformFromRefPose(self=@n2, BoneName=head, BaseName=neck_02)).Rotation)
   > Set HeadRotation(HeadRotation:Rotator=MakeRotator(Roll=@n3, Pitch=@n3.Pitch, Yaw=float * float(A=@n3.Yaw, B=0.0005))) -> (Output_Get:Rotator)
   > Set isNotPlayingEmote(isNotPlayingEmote:bool=NOT Boolean(A=Get bIsPlayingEmote())) -> (Output_Get:bool)
   > Branch(Condition=Get bIsInFrontEnd())
   > Set Custom Idle is Playing(Custom Idle is Playing:double=GetCurveValue(self=GetOwnerFrontEndAnimInstance(), CurveName=Frontend_Idle_Bypass)) -> (Output_Get:double)
   n4 = GetAnimInstance(self=GetOwnerSkeletalMeshComponent())
   Set Helmet Constraint(
    execute=@n1.then_1
    Helmet Constraint:float="To Float (Boolean)"(InBool=Montage_IsPlaying(self=@n4, Montage=/Game/Animation/Game/MainPlayer/Reacts/WorldInteraction/Hideable/PhoneBooth_Exit_Montage_V2))
   ) -> (Output_Get:float)
  ]}
]
variables [
 Custom Idle is Playing {type=number}
 HeadRotation {type=Rotator}
 Helmet Constraint {type=number}
 isNotPlayingEmote {type=boolean}
]
